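# Physics asset for the ScallopLava medium male player body cosmetic, set up for its skirt panels and waist belt ties.
# Capsules cover six two-segment skirt chains and front and back tie chains, plus pelvis, thighs and thigh armour.
# Chain segments link with tight swing and zero-degree twist limits; pelvis links are angularly free.
# base_FrontSkirt_l has no pelvis constraint, and an extra free constraint joins the two ties' second segments.
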PhysicsAsset /BRCosmetics/Characters/Player/Male/Medium/Bodies/M_MED_ScallopLava/Meshes/M_MED_ScallopLava_Physics
bodies {
 base_BackSkirt_l [{shapeType=Capsule, length=0.505, radius=0.505, shapeName=base_FrontSkirt_l_capsule}]
 base_BackSkirt_r [{shapeType=Capsule, length=0.505, radius=0.505, shapeName=base_FrontSkirt_l_capsule}]
 base_FrontSkirt_l [{shapeType=Capsule, length=0.505, radius=0.505, shapeName=base_FrontSkirt_l_capsule}]
 base_FrontSkirt_r [{shapeType=Capsule, length=0.505, radius=0.505, shapeName=base_FrontSkirt_r_capsule, rotation=<rot 0 -90 0>}]
 base_SideSkirt_l [{shapeType=Capsule, length=0.505, radius=0.505, shapeName=base_FrontSkirt_l_capsule}]
 base_SideSkirt_r [{shapeType=Capsule, length=0.505, radius=0.505, shapeName=base_FrontSkirt_l_capsule}]
 base_WaistBeltBackTie [{shapeType=Capsule, length=0, radius=1, shapeName=base_WaistBeltBackTie_capsule}]
 base_WaistBeltFrontTie [{shapeType=Capsule, length=0, radius=1, shapeName=base_WaistBeltFrontTie_capsule}]
 base_thigh_armour [{
   shapeType=Capsule
   length=14.3706
   radius=10.6163
   shapeName=base_thigh_armour_capsule
   center=<-0.4662 1.923 0.0482>
   rotation=<rot 81.8124 179.9997 179.9999>
  }]
 dyn_BackSkirt_l_00 [{shapeType=Capsule, length=7.7906, radius=2.2424, shapeName=dyn_BackSkirt_l_00_capsule, center=<3.5 0 0>, rotation=<rot 0 0 90>}]
 dyn_BackSkirt_l_01 [{shapeType=Capsule, length=9.9327, radius=2.0673, shapeName=dyn_FrontSkirt_l_01_capsule, center=<1 0 0>, rotation=<rot 90 -78.4647 0>}]
 dyn_BackSkirt_r_00 [{shapeType=Capsule, length=9.6929, radius=2.3194, shapeName=dyn_BackSkirt_r_00_capsule, center=<3.5 0 0>, rotation=<rot 0 0 90>}]
 dyn_BackSkirt_r_01 [{shapeType=Capsule, length=9.9327, radius=2.0673, shapeName=dyn_FrontSkirt_r_01_capsule, center=<1 0 0>, rotation=<rot 90 82.0659 0>}]
 dyn_FrontSkirt_l_00 [{shapeType=Capsule, length=11.2773, radius=0.9612, shapeName=dyn_FrontSkirt_l_01_capsule, center=<3 0 0>, rotation=<rot 100 60 0>}]
 dyn_FrontSkirt_l_01 [{shapeType=Capsule, length=10.6686, radius=1.3314, shapeName=dyn_FrontSkirt_l_01_capsule, center=<1 0 0>, rotation=<rot 100 60 0>}]
 dyn_FrontSkirt_r_00 [{shapeType=Capsule, length=10.9071, radius=1.3314, shapeName=dyn_FrontSkirt_r_01_capsule, center=<6 0 0>, rotation=<rot 100 -60 0>}]
 dyn_FrontSkirt_r_01 [{
   shapeType=Capsule
   length=12.0318
   radius=1.3314
   shapeName=dyn_FrontSkirt_r_01_capsule
   center=<-0.1443 1.982 -0.4035>
   rotation=<rot 80 120 180>
  }]
 dyn_SideSkirt_l_00 [{shapeType=Capsule, length=8.6505, radius=1.75, shapeName=dyn_SideSkirt_l_00_capsule, center=<4 0 0>, rotation=<rot 0 0 90>}]
 dyn_SideSkirt_l_01 [{
   shapeType=Capsule
   length=10.6686
   radius=1.3314
   shapeName=dyn_FrontSkirt_l_01_capsule
   center=<2 0 0.5>
   rotation=<rot 87.5837 68.0885 -0.0001>
  }]
 dyn_SideSkirt_r_00 [{shapeType=Capsule, length=10.1993, radius=1.75, shapeName=dyn_SideSkirt_r_00_capsule, center=<4 0 0>, rotation=<rot 0 0 90>}]
 dyn_SideSkirt_r_01 [{
   shapeType=Capsule
   length=10.6686
   radius=1.3314
   shapeName=dyn_FrontSkirt_r_01_capsule
   center=<2 0 0>
   rotation=<rot 87.5836 68.0886 144.5379>
  }]
 dyn_WaistBeltBackTie_01 [{shapeType=Capsule, length=3.5, radius=1, shapeName=dyn_WaistBeltBackTie_01_capsule, center=<3.5 -0.5 0>, rotation=<rot 90 0 0>}]
 dyn_WaistBeltBackTie_02 [{shapeType=Capsule, length=2.3465, radius=2.9035, shapeName=dyn_WaistBeltBackTie_02_capsule, center=<3 -0.25 1.0784>, rotation=<rot 90 0 0>}]
 dyn_WaistBeltBackTie_03 [{shapeType=Capsule, length=3.0125, radius=2.4875, shapeName=dyn_WaistBeltBackTie_03_capsule, center=<4 -0.125 0>, rotation=<rot 90 0 0>}]
 dyn_WaistBeltBackTie_04 [{shapeType=Capsule, length=3.049, radius=2.951, shapeName=dyn_WaistBeltBackTie_04_capsule, center=<3 0 0>, rotation=<rot 90 0 0>}]
 dyn_WaistBeltFrontTie_01 [{shapeType=Capsule, length=4, radius=1.2531, shapeName=dyn_WaistBeltFrontTie_01_capsule, center=<5 0 0.3108>, rotation=<rot 90 0 0>}]
 dyn_WaistBeltFrontTie_02 [{shapeType=Capsule, length=3.2764, radius=1.8376, shapeName=dyn_WaistBeltFrontTie_02_capsule, center=<4.0135 0 0>, rotation=<rot 90 0 0>}]
 pelvis [{shapeType=Capsule, length=7.5172, radius=11.023, shapeName=pelvis_capsule, center=<-1.4297 -0.3286 0>}]
 thigh_twist_01_l [
  {shapeType=Capsule, length=35.9419, radius=8.7729, shapeName=thigh_l_capsule, center=<5.7393 0.2755 0>, rotation=<rot 90 2.0411 0>}
  {shapeType=Capsule, length=24.7237, radius=9.95, shapeName=thigh_twist_01_l_capsule, center=<6.2667 0 11.4197>, rotation=<rot 83.0297 0 0>}
 ]
 thigh_twist_01_r [{shapeType=Capsule, length=31.2682, radius=8.7729, shapeName=thigh_r_capsule, center=<-6.457 -0.3011 0>, rotation=<rot 90 -177.9589 0>}]
}
constraints [
 {
  bone1Name=base_WaistBeltBackTie
  bone2Name=pelvis
  swing1LimitDegrees=45
  swing1Motion=Free
  swing2LimitDegrees=45
  swing2Motion=Free
  twistLimitDegrees=45
  twistMotion=Free
 }
 {bone1Name=dyn_WaistBeltBackTie_01, bone2Name=base_WaistBeltBackTie, swing1LimitDegrees=15, swing2LimitDegrees=25, twistLimitDegrees=0}
 {bone1Name=dyn_WaistBeltBackTie_02, bone2Name=dyn_WaistBeltBackTie_01, swing1LimitDegrees=15, swing2LimitDegrees=25, twistLimitDegrees=0}
 {bone1Name=dyn_WaistBeltBackTie_03, bone2Name=dyn_WaistBeltBackTie_02, swing1LimitDegrees=15, swing2LimitDegrees=25, twistLimitDegrees=0}
 {bone1Name=dyn_WaistBeltBackTie_04, bone2Name=dyn_WaistBeltBackTie_03, swing1LimitDegrees=15, swing2LimitDegrees=25, twistLimitDegrees=0}
 {
  bone1Name=base_WaistBeltFrontTie
  bone2Name=pelvis
  swing1LimitDegrees=45
  swing1Motion=Free
  swing2LimitDegrees=45
  swing2Motion=Free
  twistLimitDegrees=45
  twistMotion=Free
 }
 {bone1Name=dyn_WaistBeltFrontTie_01, bone2Name=base_WaistBeltFrontTie, swing1LimitDegrees=15, swing2LimitDegrees=25, twistLimitDegrees=0}
 {bone1Name=dyn_WaistBeltFrontTie_02, bone2Name=dyn_WaistBeltFrontTie_01, swing1LimitDegrees=15, swing2LimitDegrees=25, twistLimitDegrees=0}
 {
  bone1Name=thigh_twist_01_l
  bone2Name=pelvis
  swing1LimitDegrees=45
  swing1Motion=Free
  swing2LimitDegrees=45
  swing2Motion=Free
  twistLimitDegrees=45
  twistMotion=Free
 }
 {bone1Name=dyn_FrontSkirt_l_01, bone2Name=dyn_FrontSkirt_l_00, swing1LimitDegrees=10, swing2LimitDegrees=5, twistLimitDegrees=0}
 {bone1Name=dyn_FrontSkirt_r_01, bone2Name=dyn_FrontSkirt_r_00, swing1LimitDegrees=10, swing2LimitDegrees=5, twistLimitDegrees=0}
 {
  bone1Name=base_FrontSkirt_r
  bone2Name=pelvis
  swing1LimitDegrees=45
  swing1Motion=Free
  swing2LimitDegrees=45
  swing2Motion=Free
  twistLimitDegrees=45
  twistMotion=Free
 }
 {
  bone1Name=thigh_twist_01_r
  bone2Name=pelvis
  swing1LimitDegrees=45
  swing1Motion=Free
  swing2LimitDegrees=45
  swing2Motion=Free
  twistLimitDegrees=45
  twistMotion=Free
 }
 {
  bone1Name=dyn_WaistBeltFrontTie_02
  bone2Name=dyn_WaistBeltBackTie_02
  swing1LimitDegrees=45
  swing1Motion=Free
  swing2LimitDegrees=45
  swing2Motion=Free
  twistLimitDegrees=45
  twistMotion=Free
 }
 {
  bone1Name=base_BackSkirt_l
  bone2Name=pelvis
  swing1LimitDegrees=45
  swing1Motion=Free
  swing2LimitDegrees=45
  swing2Motion=Free
  twistLimitDegrees=45
  twistMotion=Free
 }
 {bone1Name=dyn_BackSkirt_l_01, bone2Name=dyn_BackSkirt_l_00, swing1LimitDegrees=10, swing2LimitDegrees=10, twistLimitDegrees=0}
 {
  bone1Name=base_BackSkirt_r
  bone2Name=pelvis
  swing1LimitDegrees=45
  swing1Motion=Free
  swing2LimitDegrees=45
  swing2Motion=Free
  twistLimitDegrees=45
  twistMotion=Free
 }
 {bone1Name=dyn_BackSkirt_r_01, bone2Name=dyn_BackSkirt_r_00, swing1LimitDegrees=10, swing2LimitDegrees=10, twistLimitDegrees=0}
 {
  bone1Name=base_SideSkirt_l
  bone2Name=pelvis
  swing1LimitDegrees=45
  swing1Motion=Free
  swing2LimitDegrees=45
  swing2Motion=Free
  twistLimitDegrees=45
  twistMotion=Free
 }
 {bone1Name=dyn_SideSkirt_l_01, bone2Name=dyn_SideSkirt_l_00, swing1LimitDegrees=15, swing2LimitDegrees=10, twistLimitDegrees=0}
 {
  bone1Name=base_SideSkirt_r
  bone2Name=pelvis
  swing1LimitDegrees=45
  swing1Motion=Free
  swing2LimitDegrees=45
  swing2Motion=Free
  twistLimitDegrees=45
  twistMotion=Free
 }
 {bone1Name=dyn_SideSkirt_r_01, bone2Name=dyn_SideSkirt_r_00, swing1LimitDegrees=15, swing2LimitDegrees=10, twistLimitDegrees=0}
 {bone1Name=dyn_BackSkirt_l_00, bone2Name=base_BackSkirt_l, swing1LimitDegrees=10, swing2LimitDegrees=10, twistLimitDegrees=0}
 {bone1Name=dyn_BackSkirt_r_00, bone2Name=base_BackSkirt_r, swing1LimitDegrees=10, swing2LimitDegrees=10, twistLimitDegrees=0}
 {bone1Name=dyn_SideSkirt_l_00, bone2Name=base_SideSkirt_l, swing1LimitDegrees=20, swing2LimitDegrees=10, twistLimitDegrees=0}
 {bone1Name=dyn_SideSkirt_r_00, bone2Name=base_SideSkirt_r, swing1LimitDegrees=20, swing2LimitDegrees=10, twistLimitDegrees=0}
 {
  bone1Name=base_thigh_armour
  bone2Name=pelvis
  swing1LimitDegrees=45
  swing1Motion=Free
  swing2LimitDegrees=45
  swing2Motion=Free
  twistLimitDegrees=45
  twistMotion=Free
 }
 {bone1Name=dyn_FrontSkirt_r_00, bone2Name=base_FrontSkirt_r, swing1LimitDegrees=5, swing2LimitDegrees=5, twistLimitDegrees=0}
 {bone1Name=dyn_FrontSkirt_l_00, bone2Name=base_FrontSkirt_l, swing1LimitDegrees=5, swing2LimitDegrees=5, twistLimitDegrees=0}
]
