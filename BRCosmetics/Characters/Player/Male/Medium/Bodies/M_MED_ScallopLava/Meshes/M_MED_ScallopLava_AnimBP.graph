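# Post-process style cosmetic AnimBP for the ScallopLava male body, layering secondary motion onto the copied parent mesh pose.
# A cached Control Rig pose feeds gameplay and front-end branches, selected by isInFrontEnd.
# Each branch runs pelvis-space RigidBody physics with its own asset, then additive bone offsets correcting skirt and belt ties, plus armour in gameplay.
# The EventGraph derives gravity/RBAN settings, emote blend, pistol, crouch and front-end pickaxe states.
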AnimBlueprint /BRCosmetics/Characters/Player/Male/Medium/Bodies/M_MED_ScallopLava/Meshes/M_MED_ScallopLava_AnimBP
target_skeleton=&/M_MED_ScallopLava_Skeleton
graphs [
 AnimGraph {nodes [
   n1 = RigidBody!(
    ComponentPose=Local To Component(LocalPose=Use cached pose 'CR_CtrlRig'())
    Alpha
    OverrideWorldGravity:Vector="0,0,0"
    ExternalForce
    SimSpaceSettings:Sim Space Settings Structure=(WorldAlpha=0,VelocityScaleZ=1,DampingAlpha=1,MaxLinearVelocity=10000,MaxAngularVelocity=10000,MaxLinearAcceleration=10000,MaxAngularAcceleration=10000,ExternalLinearDragV=(X=0,Y=0,Z=0),ExternalLinearVelocity=(X=0,Y=0,Z=0),ExternalAngularVelocity=(X=0,Y=0,Z=0))
   ) -> (Pose) {
    bClampLinearTranslationLimitToRefPose=true
    bOverrideWorldGravity=true
    overridePhysicsAsset=&/M_MED_ScallopLava_Physics
    simulationSpace=BaseBoneSpace
    baseBoneRef {boneName=pelvis}
   }
   n2 = "Transform (Modify) Bone - Bone: base_FrontSkirt_l"(ComponentPose=@n1, Alpha=1, Translation="0,0,0", Rotation="0.5,0,0", Scale="1,1,1") {rotationMode=BMM_Additive, rotationSpace=BCS_BoneSpace, boneToModify {boneName=base_FrontSkirt_l}, rotation {pitch=0.5}}
   n3 = "Transform (Modify) Bone - Bone: base_WaistBeltFrontTie"(ComponentPose=@n2, Alpha=1, Translation="0,0,0", Rotation="5,0,0", Scale="1,1,1") {rotationMode=BMM_Additive, rotationSpace=BCS_BoneSpace, boneToModify {boneName=base_WaistBeltFrontTie}, rotation {pitch=5}}
   n4 = "Transform (Modify) Bone - Bone: base_WaistBeltBackTie"(ComponentPose=@n3, Alpha=1, Translation="0,0,1", Rotation="2.5,0,0", Scale="1,1,1") {
    rotationMode=BMM_Additive
    rotationSpace=BCS_BoneSpace
    translationMode=BMM_Additive
    translationSpace=BCS_BoneSpace
    boneToModify {boneName=base_WaistBeltBackTie}
    rotation {pitch=2.5}
    translation {z=1}
   }
   n5 = "Transform (Modify) Bone - Bone: base_WaistBeltFrontTie"!(ComponentPose=@n4, bAlphaBoolEnabled:bool=True, Translation="0,0,0", Rotation="10,-10,0.03", Scale="1,1,1") -> (Pose) {
    alphaInputType=Bool
    rotationMode=BMM_Additive
    rotationSpace=BCS_BoneSpace
    boneToModify {boneName=base_WaistBeltFrontTie}
    rotation=<rot 10 -10 0.03>
   }
   n6 = "Transform (Modify) Bone - Bone: base_WaistBeltBackTie"!(ComponentPose=@n5, bAlphaBoolEnabled:bool=True, Translation="0,0,1", Rotation="-5,-10,0", Scale="1,1,1") -> (Pose) {
    alphaInputType=Bool
    rotationMode=BMM_Additive
    rotationSpace=BCS_BoneSpace
    translationMode=BMM_Additive
    translationSpace=BCS_BoneSpace
    boneToModify {boneName=base_WaistBeltBackTie}
    rotation {pitch=-5, yaw=-10}
    translation {z=1}
   }
   n7 = "Transform (Modify) Bone - Bone: base_WaistBeltFrontTie"(ComponentPose=@n6, bAlphaBoolEnabled=True, Translation="0,0,0", Rotation="15,0,0", Scale="1,1,1") {
    alphaInputType=Bool
    rotationMode=BMM_Additive
    rotationSpace=BCS_BoneSpace
    boneToModify {boneName=base_WaistBeltFrontTie}
    rotation {pitch=15}
   }
   n8 = "Transform (Modify) Bone - Bone: base_WaistBeltBackTie"(ComponentPose=@n7, bAlphaBoolEnabled=True, Translation="0,0,1", Rotation="10,0,0", Scale="1,1,1") {
    alphaInputType=Bool
    rotationMode=BMM_Additive
    rotationSpace=BCS_BoneSpace
    translationMode=BMM_Additive
    translationSpace=BCS_BoneSpace
    boneToModify {boneName=base_WaistBeltBackTie}
    rotation {pitch=10}
    translation {z=1}
   }
   n9 = "Transform (Modify) Bone - Bone: base_FrontSkirt_l"!(ComponentPose=@n8, bAlphaBoolEnabled:bool=True, Translation="0,0,0", Rotation="5,0,0", Scale="1,1,1") -> (Pose) {
    alphaInputType=Bool
    rotationMode=BMM_Additive
    rotationSpace=BCS_BoneSpace
    alphaBoolBlend {blendInTime=0.5, blendOutTime=0.5}
    boneToModify {boneName=base_FrontSkirt_l}
    rotation {pitch=5}
   }
   n10 = "Transform (Modify) Bone - Bone: base_FrontSkirt_r"!(ComponentPose=@n9, bAlphaBoolEnabled:bool=True, Translation="0,0,0", Rotation="8,0,0", Scale="1,1,1") -> (Pose) {
    alphaInputType=Bool
    rotationMode=BMM_Additive
    rotationSpace=BCS_BoneSpace
    alphaBoolBlend {blendInTime=0.5, blendOutTime=0.5}
    boneToModify {boneName=base_FrontSkirt_r}
    rotation {pitch=8}
   }
   n11 = "Transform (Modify) Bone - Bone: base_armour_r"!(ComponentPose=@n10, bAlphaBoolEnabled:bool=True, Translation="0,1,0.25", Rotation="0,0,0", Scale="1,1.1,1") -> (Pose) {
    alphaInputType=Bool
    rotationMode=BMM_Additive
    rotationSpace=BCS_BoneSpace
    scaleMode=BMM_Additive
    scaleSpace=BCS_BoneSpace
    translationMode=BMM_Additive
    translationSpace=BCS_BoneSpace
    boneToModify {boneName=base_armour_r}
    scale {y=1.1}
    translation {y=1, z=0.25}
   }
   n12 = "Transform (Modify) Bone - Bone: base_WaistBeltFrontTie"(ComponentPose=@n11, bAlphaBoolEnabled=True, Translation="0,0,1.157", Rotation="35.473,-35.038,4.855", Scale="1,1,1") {
    alphaInputType=Bool
    rotationMode=BMM_Additive
    rotationSpace=BCS_BoneSpace
    translationMode=BMM_Additive
    translationSpace=BCS_BoneSpace
    boneToModify {boneName=base_WaistBeltFrontTie}
    rotation=<rot 35.473 -35.038 4.855>
    translation {z=1.157}
   }
   n13 = "Transform (Modify) Bone - Bone: base_WaistBeltBackTie"(ComponentPose=@n12, bAlphaBoolEnabled=True, Translation="0,0,1", Rotation="-2.555,-30.760,-20.257", Scale="1,1,1") {
    alphaInputType=Bool
    rotationMode=BMM_Additive
    rotationSpace=BCS_BoneSpace
    translationMode=BMM_Additive
    translationSpace=BCS_BoneSpace
    boneToModify {boneName=base_WaistBeltBackTie}
    rotation=<rot -2.555 -30.76 -20.257>
    translation {z=1}
   }
   n14 = "Transform (Modify) Bone - Bone: base_WaistBeltFrontTie"(ComponentPose=@n13, bAlphaBoolEnabled=True, Translation="0,0,0", Rotation="11,-5,0", Scale="1,1,1") {
    alphaInputType=Bool
    rotationMode=BMM_Additive
    rotationSpace=BCS_BoneSpace
    translationMode=BMM_Additive
    translationSpace=BCS_BoneSpace
    alphaBoolBlend {blendInTime=0.5, blendOutTime=0.75}
    boneToModify {boneName=base_WaistBeltFrontTie}
    rotation {pitch=11, yaw=-5}
   }
   n15 = "Transform (Modify) Bone - Bone: base_WaistBeltBackTie"(ComponentPose=@n14, bAlphaBoolEnabled=True, Translation="0,0,0", Rotation="0,-15,-15", Scale="1,1,1") {
    alphaInputType=Bool
    rotationMode=BMM_Additive
    rotationSpace=BCS_BoneSpace
    translationMode=BMM_Additive
    translationSpace=BCS_BoneSpace
    alphaBoolBlend {blendInTime=0.5, blendOutTime=0.75}
    boneToModify {boneName=base_WaistBeltBackTie}
    rotation {roll=-15, yaw=-15}
   }
   n16 = "Transform (Modify) Bone - Bone: base_FrontSkirt_l"(ComponentPose=@n15, bAlphaBoolEnabled=True, Translation="0,0,0", Rotation="1,0,0", Scale="1,1,1") {
    alphaInputType=Bool
    rotationMode=BMM_Additive
    rotationSpace=BCS_BoneSpace
    translationMode=BMM_Additive
    translationSpace=BCS_BoneSpace
    alphaBoolBlend {blendInTime=0.5, blendOutTime=0.5}
    boneToModify {boneName=base_FrontSkirt_l}
    rotation {pitch=1}
   }
   n17 = "Transform (Modify) Bone - Bone: base_FrontSkirt_r"(ComponentPose=@n16, bAlphaBoolEnabled=True, Translation="0,0,0", Rotation="1,0,0", Scale="1,1,1") {
    alphaInputType=Bool
    rotationMode=BMM_Additive
    rotationSpace=BCS_BoneSpace
    translationMode=BMM_Additive
    translationSpace=BCS_BoneSpace
    alphaBoolBlend {blendInTime=0.5, blendOutTime=0.5}
    boneToModify {boneName=base_FrontSkirt_r}
    rotation {pitch=1}
   }
   n18 = RigidBody(
    ComponentPose=Local To Component(LocalPose=Use cached pose 'CR_CtrlRig'())
    OverrideWorldGravity="0,0,0"
    SimSpaceSettings=(WorldAlpha=0,VelocityScaleZ=1,DampingAlpha=1,MaxLinearVelocity=10000,MaxAngularVelocity=10000,MaxLinearAcceleration=10000,MaxAngularAcceleration=10000,ExternalLinearDragV=(X=0,Y=0,Z=0),ExternalLinearVelocity=(X=0,Y=0,Z=0),ExternalAngularVelocity=(X=0,Y=0,Z=0))
   ) {
    bClampLinearTranslationLimitToRefPose=true
    bOverrideWorldGravity=true
    overridePhysicsAsset=&/M_MED_ScallopLava_FE_Physics
    simulationSpace=BaseBoneSpace
    baseBoneRef {boneName=pelvis}
   }
   n19 = "Transform (Modify) Bone - Bone: base_FrontSkirt_l"(ComponentPose=@n18, bAlphaBoolEnabled=True, Translation="0,0,0", Rotation="-15,0,0", Scale="1,1,1") {
    alphaInputType=Bool
    rotationMode=BMM_Additive
    rotationSpace=BCS_BoneSpace
    translationMode=BMM_Additive
    translationSpace=BCS_BoneSpace
    alphaBoolBlend {blendOption=CubicInOut, blendOutTime=5}
    boneToModify {boneName=base_FrontSkirt_l}
    rotation {pitch=-15}
   }
   n20 = "Transform (Modify) Bone - Bone: base_FrontSkirt_r"(ComponentPose=@n19, bAlphaBoolEnabled=True, Translation="0,0,0", Rotation="-15,0,0", Scale="1,1,1") {
    alphaInputType=Bool
    rotationMode=BMM_Additive
    rotationSpace=BCS_BoneSpace
    translationMode=BMM_Additive
    translationSpace=BCS_BoneSpace
    alphaBoolBlend {blendOption=CubicInOut, blendOutTime=5}
    boneToModify {boneName=base_FrontSkirt_r}
    rotation {pitch=-15}
   }
   n21 = "Transform (Modify) Bone - Bone: base_WaistBeltBackTie"(ComponentPose=@n20, bAlphaBoolEnabled=True, Translation="0,0,1", Rotation="0,-3,0", Scale="1,1,1") {
    alphaInputType=Bool
    rotationMode=BMM_Additive
    rotationSpace=BCS_BoneSpace
    translationMode=BMM_Additive
    translationSpace=BCS_BoneSpace
    alphaBoolBlend {blendOutTime=2}
    boneToModify {boneName=base_WaistBeltBackTie}
    rotation {yaw=-3}
    translation {z=1}
   }
   n22 = "Transform (Modify) Bone - Bone: base_FrontSkirt_l"(
    ComponentPose=@n21
    bAlphaBoolEnabled=Get FE pickaxe on pawn() -> (FE pickaxe on pawn:bool)
    Translation="0,0,0"
    Rotation="8,0,0"
    Scale="1,1,1"
   ) {
    alphaInputType=Bool
    rotationMode=BMM_Additive
    rotationSpace=BCS_BoneSpace
    translationMode=BMM_Additive
    translationSpace=BCS_BoneSpace
    alphaBoolBlend {blendOption=CubicInOut, blendOutTime=5}
    boneToModify {boneName=base_FrontSkirt_l}
    rotation {pitch=8}
   }
   Output Pose(Result=Blend Poses by bool(
     BlendPose_0=Component To Local(ComponentPose=@n22)
     BlendPose_1=Component To Local(ComponentPose=@n17)
     bActiveValue=Get isInFrontEnd() -> (isInFrontEnd:bool)
    )) {layerGroup=None}
   n23 = Control Rig!(
    Source=Copy Pose From Mesh() {bUseAttachedParent=true}
    Alpha
    Emote Float:double=Get is Emote Float() -> (is Emote Float:double)
    isDBNO:bool=Get bIsDBNO()
    isPistolEquipped:bool=Get isPistolEquipped() -> (isPistolEquipped:bool)
    isPistolEquippedADSUP:double=Get Pistol UP() -> (Pistol UP:float)
    isSkydiving:bool=Get bIsSkyDiving()
    isDiveMode:bool=Get bIsSkydiveDiveMode()
    isParachuting:bool=Get bIsParachuteOpened()
   ) -> (Pose) {
    alphaBoolBlend {blendOption=Linear}
    controlRigAssetReference {blueprintRigClass=&/M_MED_ScallopLava_CtrlRig.M_MED_ScallopLava_CtrlRig_C}
   }
   Save cached pose 'CR_CtrlRig'(Pose=@n23) {cacheName=CR_CtrlRig}
  ]}
 EventGraph {nodes [
   n1 = Event BlueprintUpdateAnimation()
   > Set DeltaTimeX(DeltaTimeX:double=@n1.DeltaTimeX) -> (Output_Get:double)
   > n2 = Sequence!(execute) -> (then_0, then_1, then_2:Exec, then_3:Exec)
   n3 = Get bIsPlayingEmote()
   Set is Emote Float(
    execute=@n2.then_1
    is Emote Float:double=FInterpTo_Constant!(
     Current:double=Get is Emote Float()
     Target:double="To Float (Boolean)"(InBool=@n3)
     DeltaTime:double=Get DeltaTimeX() -> (DeltaTimeX:double)
     InterpSpeed:double=4
    ) -> (ReturnValue:double)
   ) -> (Output_Get:double)
   n4 = "Gravity Override State Function (Nativized)"!(
    execute=@n2.then_0
    CharacterPartAnimInstance:Custom Character Part Anim Instance Object Reference=Self-Reference()
    GravityOverrideParameterSet:Fort Gravity Override Parameter Set Structure=Get Gravity Override Params() -> (Gravity Override Params:Fort Gravity Override Parameter Set Structure)
    RigidBodyAnimNodeParameterSet:Fort Rigid Body Anim Node Parameter Set Structure=Get RBAN Params() -> (RBAN Params:Fort Rigid Body Anim Node Parameter Set Structure)
    DebugDraw:bool=false
    GameTimeInSeconds:float=GetGameTimeInSeconds()
    DebugJointName:Name=root
    StateLogicLayerName:Name=none
   ) -> (
    then
    OutGravityOverride:Vector
    OutCurrentState:Name=None
    OutGravityOverrideParameters:Fort Gravity Override Parameters Structure
    OutRigidBodyAnimNodeParameters:Fort Rigid Body Anim Node Parameters Structure
   )
   > Set Gravity Override(Gravity Override=@n4)
   > Set RBAN Settings Out(RBAN Settings Out=@n4.OutRigidBodyAnimNodeParameters)
   n5 = GetAlphaFromReferenceTransformDelta(SkeletalMeshComponent=Get SkelMesh(), BoneName=spine_05, BaseName=head, InputRange=(X=0,Y=110), TwistAxis="0,0,1")
   Set Head Rot(execute=@n2.then_3, Head Rot:float=@n5) -> (Output_Get:float)
   n6 = Get isPistolEquipped()
   > Set Pistol UP(Pistol UP:float=float * float(A=@n5, B="To Float (Boolean)"(InBool=@n6)) {is=title K2Node_PromotableOperator}) -> (Output_Get:float)
   > Set NotCrouching(NotCrouching:bool=NOT Boolean(A=Get bIsCrouching())) -> (Output_Get:bool)
   > Set RBAN_Crouching(RBAN_Crouching:bool=AND Boolean(A=Get bIsCrouching(), B=Get bRigidBodyEnabled())) -> (Output_Get:bool)
   n7 = Get bIsTargeting()
   > Set CrouchNotAiming(CrouchNotAiming:bool=AND Boolean(A=Get bIsCrouching(), B=NOT Boolean(A=@n7))) -> (Output_Get:bool)
   n8 = Get bIsPlayingEmote()
   > Set isInFrontEnd(isInFrontEnd:bool=AND Boolean!(A=Get bIsInFrontEnd(), B=Get bRigidBodyEnabled(), C:bool=NOT Boolean(A=@n8)) -> (ReturnValue)) -> (Output_Get:bool)
   n9 = TryGetFortPawn()
   > n10 = Is Valid(InputObject=@n9)
   Set FE pickaxe on pawn(execute=@n10.Is Not Valid, FE pickaxe on pawn:bool=false) -> (Output_Get:bool)
   n11 = Get CurrentWeapon!(self:Fort Pawn Object Reference=@n9) -> (CurrentWeapon:Fort Weapon Object Reference)
   n12 = IsWeaponHolstered!(self:Fort Pawn Object Reference=@n9) -> (ReturnValue:bool)
   Set FE pickaxe on pawn(execute=@n10.Is Valid, FE pickaxe on pawn=AND Boolean(A=IsValid(Object=@n11), B=NOT Boolean(A=@n12), C=Get bIsInFrontEnd()))
   n13 = Cast To FortPlayerAnimInstance(execute=@n2.then_2, Object=GetAnimInstance(self=GetOwnerSkeletalMeshComponent()))
   > Set isPistolEquipped(isPistolEquipped:bool=false) -> (Output_Get:bool)
   > n14 = Switch on EFortWeaponCoreAnimation!(execute, Selection:EFortWeaponCoreAnimation Enum=Get WeaponCoreAnim(self=@n13)) -> (
    Melee:Exec
    Pistol:Exec
    Shotgun:Exec
    PaperBlueprint:Exec
    Rifle:Exec
    MeleeOneHand:Exec
    MachinePistol:Exec
    RocketLauncher:Exec
    GrenadeLauncher:Exec
    GoingCommando:Exec
    AssaultRifle:Exec
    TacticalShotgun:Exec
    SniperRifle:Exec
    TrapPlacement:Exec
    ShoulderLauncher:Exec
    AbilityDecoTool:Exec
    Crossbow:Exec
    C4:Exec
    RemoteControl:Exec
    DualWield:Exec
    AR_BullPup:Exec
    AR_ForwardGrip:Exec
    MedPackPaddles:Exec
    SMG_P90:Exec
    AR_DrumGun:Exec
    Consumable_Small:Exec
    Consumable_Large:Exec
    Balloon:Exec
    MountedTurret:Exec
    CreativeTool:Exec
    ExplosiveBow:Exec
    AshtonIndigo:Exec
    AshtonChicago:Exec
    MeleeDualWield:Exec
    Unarmed:Exec
   )
   n15 = Branch(execute=@n14.Pistol, Condition=Get bIsTargeting())
   Set isPistolEquipped(execute=[@n15.then, @n17.else], isPistolEquipped=true)
   n16 = Branch(execute=@n15.else, Condition=Get bIsRelaxedLevel1() -> (bIsRelaxedLevel1:bool))
   Set isPistolEquipped(execute=[@n16.then, @n17.then], isPistolEquipped=false)
   n17 = Branch(execute=@n16.else, Condition=Get bIsRelaxedLevel2() -> (bIsRelaxedLevel2:bool))
   Event BlueprintInitializeAnimation()
   > Set SkelMesh(SkelMesh=GetOwnerSkeletalMeshComponent())
  ]}
]
variables [
 CrouchNotAiming {type=boolean}
 DeltaTimeX {type=number}
 FE pickaxe on pawn {type=boolean}
 Gravity Override {type=Vector}
 Gravity Override Params {
  type=FortGravityOverrideParameterSet
  default=(Skydive_GoingUp=(GravityOverride=(WindFrequency=(X=1.500000,Y=7.000000,Z=8.000000),WindAmplitude=(X=200.000000,Y=-500.000000,Z=400.000000),GravityOverride=(X=0.000000,Y=-300.000000,Z=1500.000000),JointName="root")),Skydive_FloatingDown=(GravityOverride=(WindFrequency=(X=1.500000,Y=7.000000,Z=8.000000),WindAmplitude=(X=200.000000,Y=-500.000000,Z=400.000000),WindOffset=(X=0.000000,Y=-600.000000,Z=400.000000),GravityOverride=(X=0.000000,Y=-700.000000,Z=600.000000),JointName="root")),Skydive_DivingDown=(GravityOverride=(WindFrequency=(X=1.500000,Y=12.000000,Z=10.000000),WindAmplitude=(X=50.000000,Y=300.000000,Z=100.000000),GravityOverride=(X=-1500.000000,Y=0.000000,Z=0.000000),JointName="pelvis")),Skydive_Parachuting=(GravityOverride=(WindFrequency=(X=1.500000,Y=7.000000,Z=8.000000),WindAmplitude=(X=200.000000,Y=-500.000000,Z=400.000000),GravityOverride=(X=0.000000,Y=-800.000000,Z=-980.000000),JointName="root")),OnGround_Standing=(GravityOverride=(GravityOverride=(X=0.000000,Y=0.000000,Z=-980.000000),JointName="root")),OnGround_Moving=(GravityOverride=(GravityOverride=(X=0.000000,Y=0.000000,Z=-980.000000),JointName="root")),OnGround_Moving_HighRotationRate=(GravityOverride=(GravityOverride=(X=0.000000,Y=0.000000,Z=-980.000000),JointName="root")),OnGround_Crouch_Moving=(GravityOverride=(GravityOverride=(X=0.000000,Y=0.000000,Z=-980.000000),JointName="root")),Emote_or_Melee=(GravityOverride=(GravityOverride=(X=0.000000,Y=0.000000,Z=-980.000000),JointName="root")),Swimming=(GravityOverride=(GravityOverride=(X=-800.000000,Y=0.000000,Z=0.000000),JointName="pelvis")),Front_End=(GravityOverride=(GravityOverride=(X=0.000000,Y=0.000000,Z=-800.000000),JointName="root")),DBNO=(GravityOverride=(GravityOverride=(X=-980.000000,Y=400.000000,Z=0.000000),JointName="pelvis")),DBNO_Carried=(GravityOverride=(GravityOverride=(X=-980.000000,Y=500.000000,Z=-500.000000),JointName="pelvis")),OnGround_Falling=(GravityOverride=(GravityOverride=(X=0.000000,Y=0.000000,Z=-980.000000),JointName="root")),Ziplining=(GravityOverride=(GravityOverride=(X=0.000000,Y=0.000000,Z=-980.000000),JointName="root")),Skydive_Surfing=(GravityOverride=(WindFrequency=(X=1.500000,Y=7.000000,Z=8.000000),WindAmplitude=(X=200.000000,Y=-500.000000,Z=400.000000),GravityOverride=(X=500.000000,Y=-300.000000,Z=-980.000000),JointName="root")),Riding_Standing=(GravityOverride=(GravityOverride=(X=0.000000,Y=-350.000000,Z=-980.000000),JointName="root")),Riding_Moving=(GravityOverride=(GravityOverride=(X=0.000000,Y=-450.000000,Z=-450.000000),JointName="root")),Riding_Moving_Targeting=(GravityOverride=(GravityOverride=(X=0.000000,Y=-200.000000,Z=-980.000000),JointName="root")),Tactical_Sprint=(GravityOverride=(GravityOverride=(X=0.000000,Y=-450.000000,Z=-500.000000),JointName="root")),Rail_Grinding=(GravityOverride=(GravityOverride=(X=0.000000,Y=0.000000,Z=-980.000000),JointName="root")),)
 }
 Head Rot {type=number}
 is Emote Float {type=number}
 isInFrontEnd {type=boolean}
 isPistolEquipped {type=boolean}
 NotCrouching {type=boolean}
 Pistol UP {type=number}
 RBAN Params {
  type=FortRigidBodyAnimNodeParameterSet
  default=(Skydive_GoingUp=(RigidBodyAnimNode=(SimSpaceSettings=(WorldAlpha=1.000000,MaxLinearVelocity=500.000000,MaxAngularVelocity=3.000000,MaxLinearAcceleration=800.000000,MaxAngularAcceleration=1.000000,ExternalLinearDragV=(X=1.000000,Y=1.000000,Z=1.000000)))),Skydive_FloatingDown=(RigidBodyAnimNode=(SimSpaceSettings=(WorldAlpha=1.000000,MaxLinearVelocity=500.000000,MaxAngularVelocity=3.000000,MaxLinearAcceleration=800.000000,MaxAngularAcceleration=1.000000,ExternalLinearDragV=(X=1.000000,Y=1.000000,Z=1.000000)))),Skydive_DivingDown=(RigidBodyAnimNode=(SimSpaceSettings=(WorldAlpha=1.000000,MaxLinearVelocity=500.000000,MaxAngularVelocity=3.000000,MaxLinearAcceleration=800.000000,MaxAngularAcceleration=1.000000,ExternalLinearDragV=(X=1.000000,Y=1.000000,Z=1.000000)))),Skydive_Parachuting=(RigidBodyAnimNode=(SimSpaceSettings=(WorldAlpha=1.000000,MaxLinearVelocity=500.000000,MaxAngularVelocity=3.000000,MaxLinearAcceleration=800.000000,MaxAngularAcceleration=1.000000,ExternalLinearDragV=(X=1.000000,Y=1.000000,Z=1.000000)))),OnGround_Standing=(RigidBodyAnimNode=(SimSpaceSettings=(WorldAlpha=1.000000,MaxLinearVelocity=500.000000,MaxAngularVelocity=3.000000,MaxLinearAcceleration=800.000000,MaxAngularAcceleration=1.000000,ExternalLinearDragV=(X=1.000000,Y=1.000000,Z=1.000000)))),OnGround_Moving=(RigidBodyAnimNode=(SimSpaceSettings=(WorldAlpha=1.000000,MaxLinearVelocity=500.000000,MaxAngularVelocity=3.000000,MaxLinearAcceleration=800.000000,MaxAngularAcceleration=1.000000,ExternalLinearDragV=(X=1.000000,Y=1.000000,Z=1.000000)))),OnGround_Moving_HighRotationRate=(RigidBodyAnimNode=(SimSpaceSettings=(WorldAlpha=1.000000,MaxLinearVelocity=500.000000,MaxAngularVelocity=3.000000,MaxLinearAcceleration=800.000000,MaxAngularAcceleration=1.000000,ExternalLinearDragV=(X=1.000000,Y=1.000000,Z=1.000000)))),OnGround_Crouch_Moving=(RigidBodyAnimNode=(SimSpaceSettings=(WorldAlpha=1.000000,MaxLinearVelocity=500.000000,MaxAngularVelocity=3.000000,MaxLinearAcceleration=800.000000,MaxAngularAcceleration=1.000000,ExternalLinearDragV=(X=1.000000,Y=1.000000,Z=1.000000)))),Emote_or_Melee=(RigidBodyAnimNode=(SimSpaceSettings=(WorldAlpha=0.500000,MaxLinearVelocity=500.000000,MaxAngularVelocity=3.000000,MaxLinearAcceleration=1000.000000,MaxAngularAcceleration=1.000000,ExternalLinearDragV=(X=1.000000,Y=1.000000,Z=1.000000)))),Swimming=(RigidBodyAnimNode=(SimSpaceSettings=(WorldAlpha=1.000000,MaxLinearVelocity=500.000000,MaxAngularVelocity=3.000000,MaxLinearAcceleration=800.000000,MaxAngularAcceleration=1.000000,ExternalLinearDragV=(X=1.000000,Y=1.000000,Z=1.000000)))),Front_End=(RigidBodyAnimNode=(ComponentLinearAccScale=(X=0.000000,Y=0.000000,Z=3.000000),SimSpaceSettings=(WorldAlpha=1.000000,MaxLinearVelocity=500.000000,MaxAngularVelocity=3.000000,MaxLinearAcceleration=800.000000,MaxAngularAcceleration=1.000000,ExternalLinearDragV=(X=1.000000,Y=1.000000,Z=1.000000)))),DBNO=(RigidBodyAnimNode=(SimSpaceSettings=(WorldAlpha=1.000000,MaxLinearVelocity=500.000000,MaxAngularVelocity=3.000000,MaxLinearAcceleration=800.000000,MaxAngularAcceleration=1.000000,ExternalLinearDragV=(X=1.000000,Y=1.000000,Z=1.000000)))),DBNO_Carried=(RigidBodyAnimNode=(SimSpaceSettings=(WorldAlpha=1.000000,MaxLinearVelocity=500.000000,MaxAngularVelocity=3.000000,MaxLinearAcceleration=800.000000,MaxAngularAcceleration=1.000000,ExternalLinearDragV=(X=1.000000,Y=1.000000,Z=1.000000)))),OnGround_Falling=(RigidBodyAnimNode=(SimSpaceSettings=(WorldAlpha=1.000000,MaxLinearVelocity=500.000000,MaxAngularVelocity=3.000000,MaxLinearAcceleration=800.000000,MaxAngularAcceleration=1.000000,ExternalLinearDragV=(X=1.000000,Y=1.000000,Z=1.000000)))),Ziplining=(RigidBodyAnimNode=(SimSpaceSettings=(WorldAlpha=1.000000,MaxLinearVelocity=500.000000,MaxAngularVelocity=3.000000,MaxLinearAcceleration=800.000000,MaxAngularAcceleration=1.000000,ExternalLinearDragV=(X=1.000000,Y=1.000000,Z=1.000000)))),Skydive_Surfing=(RigidBodyAnimNode=(SimSpaceSettings=(WorldAlpha=1.000000,MaxLinearVelocity=500.000000,MaxAngularVelocity=3.000000,MaxLinearAcceleration=800.000000,MaxAngularAcceleration=1.000000,ExternalLinearDragV=(X=1.000000,Y=1.000000,Z=1.000000)))),Riding_Standing=(RigidBodyAnimNode=(SimSpaceSettings=(WorldAlpha=1.000000,MaxLinearVelocity=500.000000,MaxAngularVelocity=3.000000,MaxLinearAcceleration=800.000000,MaxAngularAcceleration=1.000000,ExternalLinearDragV=(X=1.000000,Y=1.000000,Z=1.000000)))),Riding_Moving=(RigidBodyAnimNode=(SimSpaceSettings=(WorldAlpha=1.000000,MaxLinearVelocity=500.000000,MaxAngularVelocity=3.000000,MaxLinearAcceleration=800.000000,MaxAngularAcceleration=1.000000,ExternalLinearDragV=(X=1.000000,Y=1.000000,Z=1.000000)))),Riding_Moving_Targeting=(RigidBodyAnimNode=(SimSpaceSettings=(WorldAlpha=1.000000,MaxLinearVelocity=500.000000,MaxAngularVelocity=3.000000,MaxLinearAcceleration=1000.000000,MaxAngularAcceleration=1.000000,ExternalLinearDragV=(X=1.000000,Y=1.000000,Z=1.000000)))),Tactical_Sprint=(RigidBodyAnimNode=(SimSpaceSettings=(WorldAlpha=1.000000,MaxLinearVelocity=500.000000,MaxAngularVelocity=3.000000,MaxLinearAcceleration=800.000000,MaxAngularAcceleration=1.000000,ExternalLinearDragV=(X=1.000000,Y=1.000000,Z=1.000000)))),Rail_Grinding=(RigidBodyAnimNode=(SimSpaceSettings=(WorldAlpha=1.000000,MaxLinearVelocity=500.000000,MaxAngularVelocity=3.000000,MaxLinearAcceleration=800.000000,MaxAngularAcceleration=1.000000,ExternalLinearDragV=(X=1.000000,Y=1.000000,Z=1.000000)))),)
 }
 RBAN Settings Out {type=FortRigidBodyAnimNodeParameters}
 RBAN_Crouching {type=boolean}
 SkelMesh {type=SkeletalMeshComponent}
]
