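# Physics asset for the face accessory and back bling part of the male medium ScallopLava character cosmetic, covering three short hair chains and the upper body.
# Left, middle and right dyn_hair chains each link two small capsules to a base_hair body on the head, with narrow swing limits and zero twist limit.
# Capsules on upper spine, clavicles, neck and head form a freely constrained body hierarchy, with extra shapes on the right clavicle.
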PhysicsAsset /BRCosmetics/Characters/Player/Male/Medium/Bodies/M_MED_ScallopLava/Meshes/Parts/M_MED_ScallopLava_FaceAcc_BackBling_Physics
bodies {
 base_hair_l [{shapeType=Capsule, length=0.505, radius=0.505, shapeName=base_hair_l_capsule}]
 base_hair_m [{shapeType=Capsule, length=0.505, radius=0.505, shapeName=base_hair_m_capsule}]
 base_hair_r [{shapeType=Capsule, length=0.505, radius=0.505, shapeName=base_hair_r_capsule}]
 clavicle_l [{shapeType=Capsule, length=9.737, radius=6, shapeName=clavicle_l_capsule, center=<8.5246 0 3.0837>, rotation=<rot 88.4862 -0.003 -0.003>}]
 clavicle_r [
  {shapeType=Capsule, length=9.737, radius=6, shapeName=clavicle_r_capsule, center=<-8.5246 0 -3.0837>, rotation=<rot 88.4862 -0.003 179.997>}
  {shapeType=Capsule, length=2.9124, radius=11.1401, shapeName=clavicle_r_capsule, center=<-11.2815 0 -0.3566>, rotation=<rot 0 0 90>}
  {
   shapeType=Capsule
   length=0
   radius=6.6869
   shapeName=clavicle_r_capsule
   center=<-5.8521 3.6159 -6.0723>
   rotation=<rot 76.1992 -0.0001 -0.0001>
  }
 ]
 dyn_hair_l_01 [{shapeType=Capsule, length=2, radius=2, shapeName=dyn_hair_l_01_capsule, center=<2 0 0>, rotation=<rot 90 0 0>}]
 dyn_hair_l_02 [{shapeType=Capsule, length=2, radius=2, shapeName=dyn_hair_l_02_capsule, center=<2 0 0>, rotation=<rot 90 0 0>}]
 dyn_hair_m_01 [{shapeType=Capsule, length=2, radius=2, shapeName=dyn_hair_m_01_capsule, center=<2 0 0>, rotation=<rot 90 0 0>}]
 dyn_hair_m_02 [{shapeType=Capsule, length=2, radius=2, shapeName=dyn_hair_m_02_capsule, center=<2 0 0>, rotation=<rot 90 0 0>}]
 dyn_hair_r_01 [{shapeType=Capsule, length=2, radius=2, shapeName=dyn_hair_r_01_capsule, center=<2 0 0>, rotation=<rot 90 0 0>}]
 dyn_hair_r_02 [{shapeType=Capsule, length=2, radius=2, shapeName=dyn_hair_r_02_capsule, center=<2 0 0>, rotation=<rot 90 0 0>}]
 head [{shapeType=Capsule, length=0.0907, radius=5.3353, shapeName=head_capsule}]
 neck_01 [{shapeType=Capsule, length=0, radius=6.057, shapeName=neck_01_capsule}]
 neck_02 [{shapeType=Capsule, length=0, radius=6, shapeName=neck_02_capsule}]
 spine_03 [{shapeType=Capsule, length=15.0894, radius=12.9029, shapeName=spine_05_capsule, center=<3.3091 0 0>}]
 spine_04 [{shapeType=Capsule, length=18.0554, radius=12.1974, shapeName=spine_05_capsule, center=<4.6651 -1.1103 0>}]
 spine_05 [{shapeType=Capsule, length=20.0007, radius=11.2296, shapeName=spine_05_capsule, center=<3.8962 0.6633 0>}]
}
constraints [
 {
  bone1Name=clavicle_l
  bone2Name=spine_05
  swing1LimitDegrees=45
  swing1Motion=Free
  swing2LimitDegrees=45
  swing2Motion=Free
  twistLimitDegrees=45
  twistMotion=Free
 }
 {
  bone1Name=clavicle_r
  bone2Name=spine_05
  swing1LimitDegrees=45
  swing1Motion=Free
  swing2LimitDegrees=45
  swing2Motion=Free
  twistLimitDegrees=45
  twistMotion=Free
 }
 {
  bone1Name=neck_01
  bone2Name=spine_05
  swing1LimitDegrees=45
  swing1Motion=Free
  swing2LimitDegrees=45
  swing2Motion=Free
  twistLimitDegrees=45
  twistMotion=Free
 }
 {
  bone1Name=neck_02
  bone2Name=neck_01
  swing1LimitDegrees=45
  swing1Motion=Free
  swing2LimitDegrees=45
  swing2Motion=Free
  twistLimitDegrees=45
  twistMotion=Free
 }
 {
  bone1Name=head
  bone2Name=neck_02
  swing1LimitDegrees=45
  swing1Motion=Free
  swing2LimitDegrees=45
  swing2Motion=Free
  twistLimitDegrees=45
  twistMotion=Free
 }
 {
  bone1Name=base_hair_l
  bone2Name=head
  swing1LimitDegrees=45
  swing1Motion=Free
  swing2LimitDegrees=45
  swing2Motion=Free
  twistLimitDegrees=45
  twistMotion=Free
 }
 {
  bone1Name=base_hair_m
  bone2Name=head
  swing1LimitDegrees=45
  swing1Motion=Free
  swing2LimitDegrees=45
  swing2Motion=Free
  twistLimitDegrees=45
  twistMotion=Free
 }
 {
  bone1Name=base_hair_r
  bone2Name=head
  swing1LimitDegrees=45
  swing1Motion=Free
  swing2LimitDegrees=45
  swing2Motion=Free
  twistLimitDegrees=45
  twistMotion=Free
 }
 {bone1Name=dyn_hair_l_01, bone2Name=base_hair_l, swing1LimitDegrees=10, swing2LimitDegrees=15, twistLimitDegrees=0}
 {bone1Name=dyn_hair_l_02, bone2Name=dyn_hair_l_01, swing1LimitDegrees=10, swing2LimitDegrees=15, twistLimitDegrees=0}
 {bone1Name=dyn_hair_m_01, bone2Name=base_hair_m, swing1LimitDegrees=10, swing2LimitDegrees=15, twistLimitDegrees=0}
 {bone1Name=dyn_hair_m_02, bone2Name=dyn_hair_m_01, swing1LimitDegrees=10, swing2LimitDegrees=15, twistLimitDegrees=0}
 {bone1Name=dyn_hair_r_01, bone2Name=base_hair_r, swing1LimitDegrees=10, swing2LimitDegrees=15, twistLimitDegrees=0}
 {bone1Name=dyn_hair_r_02, bone2Name=dyn_hair_r_01, swing1LimitDegrees=10, swing2LimitDegrees=15, twistLimitDegrees=0}
 {
  bone1Name=spine_05
  bone2Name=spine_04
  swing1LimitDegrees=45
  swing1Motion=Free
  swing2LimitDegrees=45
  swing2Motion=Free
  twistLimitDegrees=45
  twistMotion=Free
 }
 {
  bone1Name=spine_04
  bone2Name=spine_03
  swing1LimitDegrees=45
  swing1Motion=Free
  swing2LimitDegrees=45
  swing2Motion=Free
  twistLimitDegrees=45
  twistMotion=Free
 }
]
