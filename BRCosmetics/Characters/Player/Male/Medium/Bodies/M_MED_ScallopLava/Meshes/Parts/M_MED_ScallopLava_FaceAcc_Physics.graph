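# Physics asset for the face accessory of the male medium ScallopLava cosmetic body, setting up secondary motion for hair and a tied headband.
# Capsule bodies on upper spine, clavicles, neck and head form the body volume, joined by free-rotation constraints.
# Three two-bone hair chains and a headband knot splitting into left and right three-bone tails hang from base bones with narrow swing limits.
# Extra free constraints cross-link hair tips, headband base and both tails.
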PhysicsAsset /BRCosmetics/Characters/Player/Male/Medium/Bodies/M_MED_ScallopLava/Meshes/Parts/M_MED_ScallopLava_FaceAcc_Physics
bodies {
 base_hair_l [{shapeType=Capsule, length=0.505, radius=0.505, shapeName=base_hair_l_capsule}]
 base_hair_m [{shapeType=Capsule, length=0.505, radius=0.505, shapeName=base_hair_m_capsule}]
 base_hair_r [{shapeType=Capsule, length=0.505, radius=0.505, shapeName=base_hair_r_capsule}]
 base_headband [{shapeType=Capsule, length=0.505, radius=0.505, shapeName=base_headband_capsule}]
 clavicle_l [
  {
   shapeType=Capsule
   length=5.6169
   radius=11.185
   shapeName=clavicle_l_capsule
   center=<8.3067 0 -1.8108>
   rotation=<rot 78.399 -0.0006 -0.0006>
  }
  {shapeType=Capsule, length=4.6198, radius=15, shapeName=clavicle_l_capsule, center=<17.2617 11.4762 10.8946>, rotation=<rot 90 0 0>}
 ]
 clavicle_r [
  {shapeType=Capsule, length=9.737, radius=6, shapeName=clavicle_r_capsule, center=<-8.5246 0 -3.0837>, rotation=<rot 88.4862 -0.003 179.997>}
  {shapeType=Capsule, length=7.5049, radius=11.1401, shapeName=clavicle_r_capsule, center=<-11.2815 0 -0.3566>, rotation=<rot 0 0 90>}
  {
   shapeType=Capsule
   length=3.1203
   radius=6.6869
   shapeName=clavicle_r_capsule
   center=<-7.5639 0.8943 -6.2735>
   rotation=<rot 76.199 -0.0192 -90>
  }
 ]
 dyn_hair_l_01 [{shapeType=Capsule, length=2, radius=2, shapeName=dyn_hair_l_01_capsule, center=<2 0 0>, rotation=<rot 90 0 0>}]
 dyn_hair_l_02 [{shapeType=Capsule, length=2, radius=2, shapeName=dyn_hair_l_02_capsule, center=<2 0 0>, rotation=<rot 90 0 0>}]
 dyn_hair_m_01 [{shapeType=Capsule, length=2, radius=2, shapeName=dyn_hair_m_01_capsule, center=<2 0 0>, rotation=<rot 90 0 0>}]
 dyn_hair_m_02 [{shapeType=Capsule, length=2, radius=2, shapeName=dyn_hair_m_02_capsule, center=<2 0 0>, rotation=<rot 90 0 0>}]
 dyn_hair_r_01 [{shapeType=Capsule, length=2, radius=2, shapeName=dyn_hair_r_01_capsule, center=<2 0 0>, rotation=<rot 90 0 0>}]
 dyn_hair_r_02 [{shapeType=Capsule, length=2, radius=2, shapeName=dyn_hair_r_02_capsule, center=<2 0 0>, rotation=<rot 90 0 0>}]
 dyn_headbandTie [{shapeType=Capsule, length=2.6934, radius=1.782, shapeName=dyn_headbandTie_capsule, center=<2 0 0>, rotation=<rot 90 0 0>}]
 dyn_headbandTie_l_01 [{shapeType=Capsule, length=3.4863, radius=2, shapeName=dyn_headbandTie_l_01_capsule, center=<3.2934 0 0>, rotation=<rot 90 0 0>}]
 dyn_headbandTie_l_02 [{
   shapeType=Capsule
   length=3.9589
   radius=2
   shapeName=dyn_headbandTie_l_02_capsule
   center=<3.7162 0 0>
   rotation=<rot 89.6578 -179.9956 -179.9956>
  }]
 dyn_headbandTie_l_03 [{shapeType=Capsule, length=3.323, radius=1.8803, shapeName=dyn_headbandTie_l_03_capsule, center=<3.6134 0 0>, rotation=<rot 90 0 0>}]
 dyn_headbandTie_r_01 [{shapeType=Capsule, length=3.1751, radius=2, shapeName=dyn_headbandTie_r_01_capsule, center=<3.2934 0 0>, rotation=<rot 90 0 0>}]
 dyn_headbandTie_r_02 [{shapeType=Capsule, length=3.0045, radius=2, shapeName=dyn_headbandTie_r_02_capsule, center=<3.7116 0 0>, rotation=<rot 89.1807 0 0>}]
 dyn_headbandTie_r_03 [{shapeType=Capsule, length=2, radius=2, shapeName=dyn_headbandTie_r_03_capsule, center=<2 0 0>, rotation=<rot 90 0 0>}]
 head [{shapeType=Capsule, length=0.0907, radius=5.3353, shapeName=head_capsule}]
 neck_01 [{shapeType=Capsule, length=3.9271, radius=6.057, shapeName=neck_01_capsule, center=<1.7028 -0.4275 0>}]
 neck_02 [{shapeType=Capsule, length=4.3352, radius=6.0503, shapeName=neck_02_capsule, center=<0 -0.2473 0>}]
 spine_03 [{shapeType=Capsule, length=14.3158, radius=13.6765, shapeName=spine_05_capsule, center=<3.3091 0 0>}]
 spine_04 [{shapeType=Capsule, length=17.5279, radius=12.7249, shapeName=spine_05_capsule, center=<4.6651 -1.1103 0>}]
 spine_05 [{shapeType=Capsule, length=9.5644, radius=11.6512, shapeName=spine_05_capsule, center=<3.8962 0.6633 0>}]
}
constraints [
 {
  bone1Name=clavicle_l
  bone2Name=spine_05
  swing1LimitDegrees=45
  swing1Motion=Free
  swing2LimitDegrees=45
  swing2Motion=Free
  twistLimitDegrees=45
  twistMotion=Free
 }
 {
  bone1Name=clavicle_r
  bone2Name=spine_05
  swing1LimitDegrees=45
  swing1Motion=Free
  swing2LimitDegrees=45
  swing2Motion=Free
  twistLimitDegrees=45
  twistMotion=Free
 }
 {
  bone1Name=neck_01
  bone2Name=spine_05
  swing1LimitDegrees=45
  swing1Motion=Free
  swing2LimitDegrees=45
  swing2Motion=Free
  twistLimitDegrees=45
  twistMotion=Free
 }
 {
  bone1Name=neck_02
  bone2Name=neck_01
  swing1LimitDegrees=45
  swing1Motion=Free
  swing2LimitDegrees=45
  swing2Motion=Free
  twistLimitDegrees=45
  twistMotion=Free
 }
 {
  bone1Name=head
  bone2Name=neck_02
  swing1LimitDegrees=45
  swing1Motion=Free
  swing2LimitDegrees=45
  swing2Motion=Free
  twistLimitDegrees=45
  twistMotion=Free
 }
 {
  bone1Name=base_hair_l
  bone2Name=head
  swing1LimitDegrees=45
  swing1Motion=Free
  swing2LimitDegrees=45
  swing2Motion=Free
  twistLimitDegrees=45
  twistMotion=Free
 }
 {
  bone1Name=base_hair_m
  bone2Name=head
  swing1LimitDegrees=45
  swing1Motion=Free
  swing2LimitDegrees=45
  swing2Motion=Free
  twistLimitDegrees=45
  twistMotion=Free
 }
 {bone1Name=base_headband, bone2Name=base_hair_m, swing1LimitDegrees=5, swing2LimitDegrees=5, twistLimitDegrees=0}
 {
  bone1Name=base_hair_r
  bone2Name=head
  swing1LimitDegrees=45
  swing1Motion=Free
  swing2LimitDegrees=45
  swing2Motion=Free
  twistLimitDegrees=45
  twistMotion=Free
 }
 {bone1Name=dyn_hair_l_01, bone2Name=base_hair_l, swing1LimitDegrees=10, swing2LimitDegrees=20, twistLimitDegrees=0}
 {bone1Name=dyn_hair_l_02, bone2Name=dyn_hair_l_01, swing1LimitDegrees=8, swing2LimitDegrees=25, twistLimitDegrees=0}
 {bone1Name=dyn_headbandTie, bone2Name=base_headband, swing1LimitDegrees=20, swing2LimitDegrees=20, twistLimitDegrees=0}
 {bone1Name=dyn_headbandTie_l_01, bone2Name=dyn_headbandTie, swing1LimitDegrees=25, swing2LimitDegrees=25, twistLimitDegrees=0}
 {bone1Name=dyn_headbandTie_l_02, bone2Name=dyn_headbandTie_l_01, swing1LimitDegrees=25, swing2LimitDegrees=25, twistLimitDegrees=0}
 {bone1Name=dyn_headbandTie_l_03, bone2Name=dyn_headbandTie_l_02, swing1LimitDegrees=25, swing2LimitDegrees=25, twistLimitDegrees=0}
 {bone1Name=dyn_headbandTie_r_01, bone2Name=dyn_headbandTie, swing1LimitDegrees=25, swing2LimitDegrees=25, twistLimitDegrees=0}
 {bone1Name=dyn_headbandTie_r_02, bone2Name=dyn_headbandTie_r_01, swing1LimitDegrees=25, swing2LimitDegrees=25, twistLimitDegrees=0}
 {bone1Name=dyn_headbandTie_r_03, bone2Name=dyn_headbandTie_r_02, swing1LimitDegrees=25, swing2LimitDegrees=25, twistLimitDegrees=0}
 {bone1Name=dyn_hair_m_01, bone2Name=base_hair_m, swing1LimitDegrees=10, swing2LimitDegrees=20, twistLimitDegrees=0}
 {bone1Name=dyn_hair_m_02, bone2Name=dyn_hair_m_01, swing1LimitDegrees=8, swing2LimitDegrees=25, twistLimitDegrees=0}
 {bone1Name=dyn_hair_r_01, bone2Name=base_hair_r, swing1LimitDegrees=10, swing2LimitDegrees=20, twistLimitDegrees=0}
 {bone1Name=dyn_hair_r_02, bone2Name=dyn_hair_r_01, swing1LimitDegrees=8, swing2LimitDegrees=25, twistLimitDegrees=0}
 {
  bone1Name=spine_05
  bone2Name=spine_04
  swing1LimitDegrees=45
  swing1Motion=Free
  swing2LimitDegrees=45
  swing2Motion=Free
  twistLimitDegrees=45
  twistMotion=Free
 }
 {
  bone1Name=spine_04
  bone2Name=spine_03
  swing1LimitDegrees=45
  swing1Motion=Free
  swing2LimitDegrees=45
  swing2Motion=Free
  twistLimitDegrees=45
  twistMotion=Free
 }
 {
  bone1Name=dyn_headbandTie_r_01
  bone2Name=dyn_headbandTie_l_01
  swing1LimitDegrees=45
  swing1Motion=Free
  swing2LimitDegrees=45
  swing2Motion=Free
  twistLimitDegrees=45
  twistMotion=Free
 }
 {
  bone1Name=base_headband
  bone2Name=dyn_hair_m_02
  swing1LimitDegrees=45
  swing1Motion=Free
  swing2LimitDegrees=45
  swing2Motion=Free
  twistLimitDegrees=45
  twistMotion=Free
 }
 {
  bone1Name=dyn_hair_l_02
  bone2Name=dyn_hair_m_02
  swing1LimitDegrees=45
  swing1Motion=Free
  swing2LimitDegrees=45
  swing2Motion=Free
  twistLimitDegrees=45
  twistMotion=Free
 }
 {
  bone1Name=dyn_hair_r_02
  bone2Name=dyn_hair_m_02
  swing1LimitDegrees=45
  swing1Motion=Free
  swing2LimitDegrees=45
  swing2Motion=Free
  twistLimitDegrees=45
  twistMotion=Free
 }
]
